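# M_Floor_01 is a Default Lit floor material for the VR Editor basic meshes, supporting static lighting.
# Vertex colors act as masks: green blends trim into floor color, and squared red blends in groove color and scales a glow color.
# A tiled grunge texture's blue channel supplies a roughness-style value, which green blends from raw toward a remap between Rough Min and Rough Max.
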Material /Engine/VREditor/BasicMeshes/M_Floor_01
graph {
 properties {bUsedWithStaticLighting=true, pixelDepthOffsetMode=PDOM_Legacy, usedShadingModels=Default Lit}
 nodes [
  n1 = TextureSample(UVs=TextureCoordinate() {uTiling=6, vTiling=6}) {texture=&/T_CleanFloorGrunge_01_D}
  n2 = VertexColor()
  LinearInterpolate(
   A=@n1.B
   B=LinearInterpolate(
    A=ScalarParameter() {defaultValue=0.1, parameterName=Rough Min}
    B=ScalarParameter() {defaultValue=0.34, parameterName=Rough Max}
    Alpha=@n1.B
   )
   Alpha=@n2.G
  )
  n3 = Multiply(A=@n2.R, B=@n2.R)
  LinearInterpolate(
   A=LinearInterpolate(
    A=VectorParameter() {parameterName=TrimColor, defaultValue=<rgba 0.03 0.03 0.03 1>}.RGB
    B=VectorParameter() {parameterName=FloorColor, defaultValue=<rgba 0.06 0.06 0.06 1>}.RGB
    Alpha=@n2.G
   )
   B=VectorParameter() {parameterName=GrooveColor, defaultValue=<rgba 0 0 0 1>}.RGB
   Alpha=@n3
  )
  Multiply(A=VectorParameter() {parameterName=GlowColor, defaultValue=<rgba 0 0 0 1>}.RGB, B=@n3)
 ]
}
